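# Post-process-style AnimBlueprint for the GoldenValley backpack, adding secondary motion to its hanging skates, wheels, tags, tassels and handle.
# A chain of spring-driven Anim Dynamics nodes simulates the input pose, then additive bone rotations adjust pack, skates and tassels.
# The root gets a small offset for female characters, detected from customization part gender.
# The event graph also sets IsinAir from skydive states, though no shown node reads it.
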AnimBlueprint /BRCosmetics/Accessories/FORT_Backpacks/Backpack_F_MED_GoldenValley/Meshes/F_MED_GoldenValley_Pack_AnimBP
target_skeleton=&/F_MED_GoldenValley_Pack_Skeleton
graphs [
 AnimGraph {nodes [
   n1 = Input Pose()
   n2 = Anim Dynamics - Bone: dyn_R_skate(
    ComponentPose="Transform (Modify) Bone - Bone: root"!(
     ComponentPose=Local To Component(LocalPose=@n1)
     bAlphaBoolEnabled:bool=Get isCharFemale() -> (isCharFemale:bool)
     Translation="0,2,0"
     Rotation="0,0,0"
     Scale
    ) -> (Pose) {alphaInputType=Bool, translationMode=BMM_Additive, boneToModify {boneName=root}, translation {y=2}}
    Alpha=1
    ExternalForce="0,0,0"
   ) {
    angularDampingOverride=0.99
    angularSpringConstant=3000
    bAngularSpring=true
    bOverrideAngularDamping=true
    bUseGravityOverride=true
    boundBone {boneName=dyn_R_skate}
    componentAppliedLinearAccClamp=<800 800 900>
    componentLinearAccScale=<0.4 0.4 0.4>
    componentLinearVelScale=<0.5 0.5 0.5>
    gravityOverride {z=-980}
    physicsBodyDefinitions [{
      boundBone {boneName=dyn_R_skate}
      boxExtents=<20 15 16>
      constraintSetup {angularLimitsMax=<5 50 3>, angularLimitsMin=<-5 -3 0>, angularTarget=<1 0 0>}
      localJointOffset=<-0.85 0.2 0.1>
     }]
    relativeSpaceBone {boneName=pack}
   }
   n3 = Anim Dynamics - Bone: dyn_R_wheel_01(ComponentPose=@n2, Alpha=1, ExternalForce="0,0,0") {
    angularDampingOverride=1
    angularSpringConstant=3000
    bAngularSpring=true
    bOverrideAngularDamping=true
    bUseGravityOverride=true
    boundBone {boneName=dyn_R_wheel_01}
    componentAppliedLinearAccClamp=<1200 1200 1200>
    componentLinearAccScale=<0.5 0.5 0.5>
    componentLinearVelScale=<0.6 0.6 0.6>
    gravityOverride {z=-980}
    physicsBodyDefinitions [{
      boundBone {boneName=dyn_R_wheel_01}
      boxExtents=<7 7 7>
      constraintSetup {angularTargetAxis=AxisY, twistAxis=AxisY, angularLimitsMax=<90 0 0>, angularLimitsMin=<-90 0 0>, angularTarget=<0 1 0>}
      localJointOffset=<0 0.5 0>
     }]
    relativeSpaceBone {boneName=dyn_R_wheel_01}
   }
   n4 = Anim Dynamics - Bone: dyn_R_wheel_02(ComponentPose=@n3, Alpha=1, ExternalForce="0,0,0") {
    angularDampingOverride=1
    angularSpringConstant=3000
    bAngularSpring=true
    bOverrideAngularDamping=true
    bUseGravityOverride=true
    boundBone {boneName=dyn_R_wheel_02}
    componentAppliedLinearAccClamp=<1200 1200 1200>
    componentLinearAccScale=<0.5 0.5 0.5>
    componentLinearVelScale=<0.6 0.6 0.6>
    gravityOverride {z=-980}
    physicsBodyDefinitions [{
      boundBone {boneName=dyn_R_wheel_02}
      boxExtents=<8 7 8>
      constraintSetup {angularTargetAxis=AxisY, twistAxis=AxisY, angularLimitsMax=<90 0 0>, angularLimitsMin=<-90 0 0>, angularTarget=<0 1 0>}
      localJointOffset=<0 0.5 0>
     }]
    relativeSpaceBone {boneName=dyn_R_wheel_01}
   }
   n5 = Anim Dynamics - Bone: dyn_R_wheel_03(ComponentPose=@n4, Alpha=1, ExternalForce="0,0,0") {
    angularDampingOverride=1
    angularSpringConstant=1200
    bAngularSpring=true
    bOverrideAngularDamping=true
    bUseGravityOverride=true
    boundBone {boneName=dyn_R_wheel_03}
    componentAppliedLinearAccClamp=<1200 1200 1200>
    componentLinearAccScale=<0.5 0.5 0.5>
    componentLinearVelScale=<0.6 0.6 0.6>
    gravityOverride {z=-980}
    physicsBodyDefinitions [{
      boundBone {boneName=dyn_R_wheel_03}
      boxExtents=<7 9 9>
      constraintSetup {angularTargetAxis=AxisY, twistAxis=AxisY, angularLimitsMax=<90 0 0>, angularLimitsMin=<-90 0 0>, angularTarget=<0 1 0>}
      localJointOffset=<0 0.4 0>
     }]
    relativeSpaceBone {boneName=dyn_R_wheel_01}
   }
   n6 = Anim Dynamics - Bone: dyn_R_wheel_04(ComponentPose=@n5, Alpha=1, ExternalForce="0,0,0") {
    angularDampingOverride=1
    angularSpringConstant=1000
    bAngularSpring=true
    bOverrideAngularDamping=true
    bUseGravityOverride=true
    boundBone {boneName=dyn_R_wheel_04}
    componentAppliedLinearAccClamp=<1200 1200 1200>
    componentLinearAccScale=<0.5 0.5 0.5>
    componentLinearVelScale=<0.6 0.6 0.6>
    gravityOverride {z=-980}
    physicsBodyDefinitions [{
      boundBone {boneName=dyn_R_wheel_04}
      boxExtents=<6 7 7>
      constraintSetup {angularTargetAxis=AxisY, twistAxis=AxisY, angularLimitsMax=<90 0 0>, angularLimitsMin=<-90 0 0>, angularTarget=<0 1 0>}
      localJointOffset=<0 0.6 0>
     }]
    relativeSpaceBone {boneName=dyn_R_wheel_01}
   }
   n7 = Anim Dynamics - Bone: dyn_R_skateTag(ComponentPose=@n6, Alpha=1, ExternalForce="0,0,0") {
    angularDampingOverride=0.999
    angularSpringConstant=4000
    bAngularSpring=true
    bOverrideAngularDamping=true
    bUseGravityOverride=true
    boundBone {boneName=dyn_R_skateTag}
    componentAppliedLinearAccClamp=<1000 1000 800>
    componentLinearAccScale=<0.3 0.3 0.3>
    componentLinearVelScale=<0.4 0.4 0.4>
    gravityOverride {z=-980}
    physicsBodyDefinitions [{
      boundBone {boneName=dyn_R_skateTag}
      boxExtents=<8 7 10>
      constraintSetup {angularLimitsMax=<0 40 0>, angularLimitsMin=<0 -40 0>, angularTarget=<1 0 0>}
      localJointOffset=<0.7 0 0>
     }]
   }
   n8 = Anim Dynamics - Bone: dyn_L_skate(ComponentPose=@n7, Alpha=1, ExternalForce="0,0,0") {
    angularDampingOverride=0.999
    angularSpringConstant=2700
    bAngularSpring=true
    bOverrideAngularDamping=true
    bUseGravityOverride=true
    boundBone {boneName=dyn_L_skate}
    componentAppliedLinearAccClamp=<800 800 900>
    componentLinearAccScale=<0.4 0.4 0.4>
    componentLinearVelScale=<0.5 0.5 0.5>
    gravityOverride {z=-980}
    physicsBodyDefinitions [{
      boundBone {boneName=dyn_L_skate}
      boxExtents=<16 15 16>
      constraintSetup {angularLimitsMax=<5 40 3>, angularLimitsMin=<-5 -5 0>, angularTarget=<1 0 0>}
      localJointOffset=<0.7 0.4 0>
     }]
   }
   n9 = Anim Dynamics - Bone: dyn_L_wheel_01(ComponentPose=@n8, Alpha=1, ExternalForce="0,0,0") {
    angularDampingOverride=1
    angularSpringConstant=3000
    bAngularSpring=true
    bOverrideAngularDamping=true
    bUseGravityOverride=true
    boundBone {boneName=dyn_L_wheel_01}
    componentAppliedLinearAccClamp=<1200 1200 1200>
    componentLinearAccScale=<0.5 0.5 0.5>
    componentLinearVelScale=<0.6 0.6 0.6>
    gravityOverride {z=-980}
    physicsBodyDefinitions [{
      boundBone {boneName=dyn_L_wheel_01}
      boxExtents=<7 7 7>
      constraintSetup {angularTargetAxis=AxisY, twistAxis=AxisY, angularLimitsMax=<90 0 0>, angularLimitsMin=<-90 0 0>, angularTarget=<0 1 0>}
      localJointOffset=<0 -0.5 0>
     }]
    relativeSpaceBone {boneName=dyn_R_wheel_01}
   }
   n10 = Anim Dynamics - Bone: dyn_L_wheel_02(ComponentPose=@n9, Alpha=1, ExternalForce="0,0,0") {
    angularDampingOverride=1
    angularSpringConstant=3000
    bAngularSpring=true
    bOverrideAngularDamping=true
    bUseGravityOverride=true
    boundBone {boneName=dyn_L_wheel_02}
    componentAppliedLinearAccClamp=<1200 1200 1200>
    componentLinearAccScale=<0.5 0.5 0.5>
    componentLinearVelScale=<0.6 0.6 0.6>
    gravityOverride {z=-980}
    physicsBodyDefinitions [{
      boundBone {boneName=dyn_L_wheel_02}
      boxExtents=<8 7 8>
      constraintSetup {angularTargetAxis=AxisY, twistAxis=AxisY, angularLimitsMax=<90 0 0>, angularLimitsMin=<-90 0 0>, angularTarget=<0 1 0>}
      localJointOffset=<0 -0.5 0>
     }]
    relativeSpaceBone {boneName=dyn_R_wheel_01}
   }
   n11 = Anim Dynamics - Bone: dyn_L_wheel_03(ComponentPose=@n10, Alpha=1, ExternalForce="0,0,0") {
    angularDampingOverride=1
    angularSpringConstant=1000
    bAngularSpring=true
    bOverrideAngularDamping=true
    bUseGravityOverride=true
    boundBone {boneName=dyn_L_wheel_03}
    componentAppliedLinearAccClamp=<1200 1200 1200>
    componentLinearAccScale=<0.5 0.5 0.5>
    componentLinearVelScale=<0.6 0.6 0.6>
    gravityOverride {z=-980}
    physicsBodyDefinitions [{
      boundBone {boneName=dyn_L_wheel_03}
      boxExtents=<7 9 9>
      constraintSetup {angularTargetAxis=AxisY, twistAxis=AxisY, angularLimitsMax=<90 0 0>, angularLimitsMin=<-90 0 0>, angularTarget=<0 1 0>}
      localJointOffset=<0 -0.4 0>
     }]
    relativeSpaceBone {boneName=dyn_R_wheel_01}
   }
   n12 = Anim Dynamics - Bone: dyn_L_wheel_04(ComponentPose=@n11, Alpha=1, ExternalForce="0,0,0") {
    angularDampingOverride=1
    angularSpringConstant=1200
    bAngularSpring=true
    bOverrideAngularDamping=true
    bUseGravityOverride=true
    boundBone {boneName=dyn_L_wheel_04}
    componentAppliedLinearAccClamp=<1200 1200 1200>
    componentLinearAccScale=<0.5 0.5 0.5>
    componentLinearVelScale=<0.6 0.6 0.6>
    gravityOverride {z=-980}
    physicsBodyDefinitions [{
      boundBone {boneName=dyn_L_wheel_04}
      boxExtents=<6 7 7>
      constraintSetup {angularTargetAxis=AxisY, twistAxis=AxisY, angularLimitsMax=<90 0 0>, angularLimitsMin=<-90 0 0>, angularTarget=<0 1 0>}
      localJointOffset=<0 -0.6 0>
     }]
    relativeSpaceBone {boneName=dyn_R_wheel_01}
   }
   n13 = Anim Dynamics - Bone: dyn_L_skateTag(ComponentPose=@n12, Alpha=1, ExternalForce="0,0,0") {
    angularDampingOverride=0.999
    angularSpringConstant=4500
    bAngularSpring=true
    bOverrideAngularDamping=true
    bUseGravityOverride=true
    boundBone {boneName=dyn_L_skateTag}
    componentAppliedLinearAccClamp=<1000 1000 1000>
    componentLinearAccScale=<0.3 0.3 0.3>
    componentLinearVelScale=<0.4 0.4 0.4>
    gravityOverride {z=-980}
    physicsBodyDefinitions [{
      boundBone {boneName=dyn_L_skateTag}
      boxExtents=<10 10 10>
      constraintSetup {angularLimitsMax=<0 20 0>, angularLimitsMin=<0 -20 0>, angularTarget=<1 0 0>}
      localJointOffset=<-0.5 0 0>
     }]
   }
   n14 = Anim Dynamics - Bone: dyn_tassel_01_A1(ComponentPose=@n13, Alpha=1, ExternalForce="0,0,0") {
    angularDampingOverride=0.999
    angularSpringConstant=800
    bAngularSpring=true
    bOverrideAngularDamping=true
    bUseGravityOverride=true
    boundBone {boneName=dyn_tassel_01_A1}
    componentAppliedLinearAccClamp=<700 700 700>
    componentLinearAccScale=<0.4 0.4 0.4>
    componentLinearVelScale=<0.5 0.5 0.5>
    gravityOverride {z=-980}
    physicsBodyDefinitions [{
      boundBone {boneName=dyn_tassel_01_A1}
      boxExtents=<15 8 8>
      constraintSetup {angularLimitsMax=<0 30 5>, angularLimitsMin=<0 0 -5>, angularTarget=<0 0 0>}
      localJointOffset=<1 0.3 0>
     }]
   }
   n15 = Anim Dynamics - Bone: dyn_tassel_01_B(ComponentPose=@n14, Alpha=1, ExternalForce="0,0,0") {
    angularDampingOverride=0.999
    angularSpringConstant=1100
    bAngularSpring=true
    bOverrideAngularDamping=true
    bUseGravityOverride=true
    boundBone {boneName=dyn_tassel_01_B}
    componentAppliedLinearAccClamp=<850 850 850>
    componentLinearAccScale=<0.4 0.4 0.4>
    componentLinearVelScale=<0.5 0.5 0.5>
    gravityOverride {z=-980}
    physicsBodyDefinitions [{
      boundBone {boneName=dyn_tassel_01_B}
      boxExtents=<15 8 8>
      constraintSetup {angularLimitsMax=<0 50 10>, angularLimitsMin=<0 0 -10>, angularTarget=<0 0 0>}
      localJointOffset=<1.4 0 0>
     }]
   }
   n16 = Anim Dynamics - Bone: dyn_tassel_02_A(ComponentPose=@n15, Alpha=1, ExternalForce="0,0,0") {
    angularDampingOverride=0.999
    angularSpringConstant=1000
    bAngularSpring=true
    bOverrideAngularDamping=true
    bUseGravityOverride=true
    boundBone {boneName=dyn_tassel_02_A}
    componentAppliedLinearAccClamp=<700 700 700>
    componentLinearAccScale=<0.4 0.4 0.4>
    componentLinearVelScale=<0.5 0.5 0.5>
    gravityOverride {z=-980}
    physicsBodyDefinitions [{
      boundBone {boneName=dyn_tassel_02_A}
      boxExtents=<15 8 8>
      constraintSetup {angularLimitsMax=<0 30 5>, angularLimitsMin=<0 0 -5>, angularTarget=<0 0 0>}
      localJointOffset=<1 0 0>
     }]
   }
   n17 = Anim Dynamics - Bone: dyn_tassel_02_B(ComponentPose=@n16, Alpha=1, ExternalForce="0,0,0") {
    angularDampingOverride=0.999
    angularSpringConstant=1200
    bAngularSpring=true
    bOverrideAngularDamping=true
    bUseGravityOverride=true
    boundBone {boneName=dyn_tassel_02_B}
    componentAppliedLinearAccClamp=<850 850 850>
    componentLinearAccScale=<0.4 0.4 0.4>
    componentLinearVelScale=<0.5 0.5 0.5>
    gravityOverride {z=-980}
    physicsBodyDefinitions [{
      boundBone {boneName=dyn_tassel_02_B}
      boxExtents=<12 8 8>
      constraintSetup {angularLimitsMax=<0 50 10>, angularLimitsMin=<0 0 -10>, angularTarget=<0 0 0>}
      localJointOffset=<1.4 0 0.1>
     }]
   }
   n18 = Anim Dynamics - Bone: dyn_tassel_03_A(ComponentPose=@n17, Alpha=1, ExternalForce="0,0,0") {
    angularDampingOverride=0.999
    angularSpringConstant=1000
    bAngularSpring=true
    bOverrideAngularDamping=true
    bUseGravityOverride=true
    boundBone {boneName=dyn_tassel_03_A}
    componentAppliedLinearAccClamp=<700 700 700>
    componentLinearAccScale=<0.4 0.4 0.4>
    componentLinearVelScale=<0.5 0.5 0.5>
    gravityOverride {z=-980}
    physicsBodyDefinitions [{
      boundBone {boneName=dyn_tassel_03_A}
      boxExtents=<15 8 8>
      constraintSetup {angularLimitsMax=<0 30 5>, angularLimitsMin=<0 0 -5>, angularTarget=<0 0 0>}
      localJointOffset=<1 0.3 0>
     }]
   }
   n19 = Anim Dynamics - Bone: dyn_tassel_03_B(ComponentPose=@n18, Alpha=1, ExternalForce="0,0,0") {
    angularDampingOverride=0.9999
    angularSpringConstant=1500
    bAngularSpring=true
    bOverrideAngularDamping=true
    bUseGravityOverride=true
    boundBone {boneName=dyn_tassel_03_B}
    componentAppliedLinearAccClamp=<850 850 850>
    componentLinearAccScale=<0.4 0.4 0.4>
    componentLinearVelScale=<0.5 0.5 0.5>
    gravityOverride {z=-980}
    physicsBodyDefinitions [{
      boundBone {boneName=dyn_tassel_03_B}
      boxExtents=<15 8 8>
      constraintSetup {angularLimitsMax=<3 50 10>, angularLimitsMin=<-3 0 -10>, angularTarget=<0 0 0>}
      localJointOffset=<1.4 0 0>
     }]
   }
   n20 = Anim Dynamics - Bone: dyn_tassel_07_A(ComponentPose=@n19, Alpha=1, ExternalForce="0,0,0") {
    angularDampingOverride=0.999
    angularSpringConstant=1000
    bAngularSpring=true
    bOverrideAngularDamping=true
    bUseGravityOverride=true
    boundBone {boneName=dyn_tassel_07_A}
    componentAppliedLinearAccClamp=<700 700 700>
    componentLinearAccScale=<0.4 0.4 0.4>
    componentLinearVelScale=<0.5 0.5 0.5>
    gravityOverride {z=-980}
    physicsBodyDefinitions [{
      boundBone {boneName=dyn_tassel_07_A}
      boxExtents=<15 8 8>
      constraintSetup {angularLimitsMax=<0 30 5>, angularLimitsMin=<0 0 -5>, angularTarget=<0 0 0>}
      localJointOffset=<1 0 0>
     }]
   }
   n21 = Anim Dynamics - Bone: dyn_tassel_07_B(ComponentPose=@n20, Alpha=1, ExternalForce="0,0,0") {
    angularDampingOverride=1
    angularSpringConstant=1500
    bAngularSpring=true
    bOverrideAngularDamping=true
    bUseGravityOverride=true
    boundBone {boneName=dyn_tassel_07_B}
    componentAppliedLinearAccClamp=<800 800 850>
    componentLinearAccScale=<0.4 0.4 0.4>
    componentLinearVelScale=<0.5 0.5 0.5>
    gravityOverride {z=-980}
    physicsBodyDefinitions [{
      boundBone {boneName=dyn_tassel_07_B}
      boxExtents=<15 10 8>
      constraintSetup {angularLimitsMax=<5 50 15>, angularLimitsMin=<-5 0 -15>, angularTarget=<0 0 0>}
      localJointOffset=<1.4 0 0.1>
     }]
   }
   n22 = Anim Dynamics - Bone: dyn_tassel_04_A(ComponentPose=@n21, Alpha=1, ExternalForce="0,0,0") {
    angularDampingOverride=0.999
    angularSpringConstant=1000
    bAngularSpring=true
    bOverrideAngularDamping=true
    bUseGravityOverride=true
    boundBone {boneName=dyn_tassel_04_A}
    componentAppliedLinearAccClamp=<700 700 700>
    componentLinearAccScale=<0.4 0.4 0.4>
    componentLinearVelScale=<0.5 0.5 0.5>
    gravityOverride {z=-980}
    physicsBodyDefinitions [{
      boundBone {boneName=dyn_tassel_04_A}
      boxExtents=<15 8 8>
      constraintSetup {angularLimitsMax=<0 30 5>, angularLimitsMin=<0 0 -5>, angularTarget=<0 0 0>}
      localJointOffset=<1.1 0.1 0>
     }]
   }
   n23 = Anim Dynamics - Bone: dyn_tassel_04_B(ComponentPose=@n22, Alpha=1, ExternalForce="0,0,0") {
    angularDampingOverride=0.99
    angularSpringConstant=1100
    bAngularSpring=true
    bOverrideAngularDamping=true
    bUseGravityOverride=true
    boundBone {boneName=dyn_tassel_04_B}
    componentAppliedLinearAccClamp=<850 850 800>
    componentLinearAccScale=<0.4 0.4 0.3>
    componentLinearVelScale=<0.5 0.5 0.4>
    gravityOverride {z=-980}
    physicsBodyDefinitions [{
      boundBone {boneName=dyn_tassel_04_B}
      boxExtents=<12 8 12>
      constraintSetup {angularLimitsMax=<3 50 21>, angularLimitsMin=<-3 0 -12>, angularTarget=<0 0 0>}
      localJointOffset=<1.5 0 0>
     }]
   }
   n24 = Anim Dynamics - Bone: dyn_tassel_05_A(ComponentPose=@n23, Alpha=1, ExternalForce="0,0,0") {
    angularDampingOverride=0.999
    angularSpringConstant=1000
    bAngularSpring=true
    bOverrideAngularDamping=true
    bUseGravityOverride=true
    boundBone {boneName=dyn_tassel_05_A}
    componentAppliedLinearAccClamp=<700 700 700>
    componentLinearAccScale=<0.4 0.4 0.4>
    componentLinearVelScale=<0.5 0.5 0.5>
    gravityOverride {z=-980}
    physicsBodyDefinitions [{
      boundBone {boneName=dyn_tassel_05_A}
      boxExtents=<15 8 8>
      constraintSetup {angularLimitsMax=<0 30 5>, angularLimitsMin=<0 0 -5>, angularTarget=<0 0 0>}
      localJointOffset=<1 0 0>
     }]
   }
   n25 = Anim Dynamics - Bone: dyn_tassel_05_B(ComponentPose=@n24, Alpha=1, ExternalForce="0,0,0") {
    angularDampingOverride=0.9999
    angularSpringConstant=1200
    bAngularSpring=true
    bOverrideAngularDamping=true
    bUseGravityOverride=true
    boundBone {boneName=dyn_tassel_05_B}
    componentAppliedLinearAccClamp=<850 850 850>
    componentLinearAccScale=<0.4 0.4 0.4>
    componentLinearVelScale=<0.5 0.5 0.5>
    gravityOverride {z=-980}
    physicsBodyDefinitions [{
      boundBone {boneName=dyn_tassel_05_B}
      boxExtents=<15 10 8>
      constraintSetup {angularLimitsMax=<0 50 10>, angularLimitsMin=<0 0 -10>, angularTarget=<0 0 0>}
      localJointOffset=<1.6 0.1 0.1>
     }]
   }
   n26 = Anim Dynamics - Bone: dyn_tassel_06_B(ComponentPose=@n25, Alpha=1, ExternalForce="0,0,0") {
    angularDampingOverride=0.999
    angularSpringConstant=800
    bAngularSpring=true
    bOverrideAngularDamping=true
    bUseGravityOverride=true
    boundBone {boneName=dyn_tassel_06_B}
    componentAppliedLinearAccClamp=<700 700 700>
    componentLinearAccScale=<0.4 0.4 0.4>
    componentLinearVelScale=<0.5 0.5 0.5>
    gravityOverride {z=-980}
    physicsBodyDefinitions [{
      boundBone {boneName=dyn_tassel_06_B}
      boxExtents=<15 8 8>
      constraintSetup {angularLimitsMax=<0 30 5>, angularLimitsMin=<0 0 -5>, angularTarget=<0 0 0>}
      localJointOffset=<1 0 0>
     }]
   }
   n27 = Anim Dynamics - Bone: dyn_tassel_07_B(ComponentPose=@n26, Alpha=1, ExternalForce="0,0,0") {
    angularDampingOverride=0.999
    angularSpringConstant=1100
    bAngularSpring=true
    bOverrideAngularDamping=true
    bUseGravityOverride=true
    boundBone {boneName=dyn_tassel_07_B}
    componentAppliedLinearAccClamp=<850 850 850>
    componentLinearAccScale=<0.4 0.4 0.4>
    componentLinearVelScale=<0.5 0.5 0.5>
    gravityOverride {z=-980}
    physicsBodyDefinitions [{
      boundBone {boneName=dyn_tassel_07_B}
      boxExtents=<15 8 8>
      constraintSetup {angularLimitsMax=<0 50 10>, angularLimitsMin=<0 0 -10>, angularTarget=<0 0 0>}
      localJointOffset=<1.4 0 0>
     }]
   }
   n28 = Anim Dynamics - Bone: dyn_handle(ComponentPose=@n27, Alpha=1, ExternalForce="0,0,0") {
    angularDampingOverride=1
    angularSpringConstant=4000
    bAngularSpring=true
    bOverrideAngularDamping=true
    bUseGravityOverride=true
    boundBone {boneName=dyn_handle}
    componentAppliedLinearAccClamp=<800 800 900>
    componentLinearAccScale=<0.5 0.5 0.5>
    componentLinearVelScale=<0.6 0.6 0.6>
    gravityOverride {z=-980}
    physicsBodyDefinitions [{
      boundBone {boneName=dyn_handle}
      boxExtents=<20 15 16>
      constraintSetup {angularLimitsMax=<0 55 3>, angularLimitsMin=<0 -45 0>, angularTarget=<1 0 0>}
      localJointOffset=<2 0.2 0>
     }]
   }
   n29 = "Transform (Modify) Bone - Bone: pack"!(ComponentPose=@n28, bAlphaBoolEnabled:bool, Translation="0,0,0", Rotation="0,0,8", Scale) -> (Pose) {
    alphaInputType=Bool
    rotationMode=BMM_Additive
    rotationSpace=BCS_ParentBoneSpace
    translationMode=BMM_Additive
    translationSpace=BCS_ParentBoneSpace
    alphaBoolBlend {blendInTime=0.2, blendOutTime=0.2}
    boneToModify {boneName=pack}
    rotation {roll=8}
   }
   n30 = "Transform (Modify) Bone - Bone: dyn_R_skate"!(ComponentPose=@n29, bAlphaBoolEnabled:bool=True, Translation="0,0,0", Rotation="0,0,40", Scale="1,1,1") -> (Pose) {
    alphaInputType=Bool
    rotationMode=BMM_Additive
    rotationSpace=BCS_ParentBoneSpace
    translationMode=BMM_Additive
    translationSpace=BCS_ParentBoneSpace
    alphaBoolBlend {blendInTime=0.2, blendOutTime=0.2}
    boneToModify {boneName=dyn_R_skate}
    rotation {roll=40}
   }
   n31 = "Transform (Modify) Bone - Bone: dyn_L_skate"!(ComponentPose=@n30, bAlphaBoolEnabled:bool=True, Translation="0,0,0", Rotation="-0.000,0.000,45.000", Scale="1,1,1") -> (Pose) {
    alphaInputType=Bool
    rotationMode=BMM_Additive
    rotationSpace=BCS_ParentBoneSpace
    translationMode=BMM_Additive
    translationSpace=BCS_ParentBoneSpace
    alphaBoolBlend {blendInTime=0.2, blendOutTime=0.2}
    boneToModify {boneName=dyn_L_skate}
    rotation {roll=45}
   }
   n32 = "Transform (Modify) Bone - Bone: dyn_tassel_07_A"!(ComponentPose=@n31, bAlphaBoolEnabled:bool=True, Translation="0,0,0", Rotation="0,0,50", Scale="1,1,1") -> (Pose) {
    alphaInputType=Bool
    rotationMode=BMM_Additive
    rotationSpace=BCS_ParentBoneSpace
    translationMode=BMM_Additive
    translationSpace=BCS_ParentBoneSpace
    alphaBoolBlend {blendInTime=0.2, blendOutTime=0.2}
    boneToModify {boneName=dyn_tassel_07_A}
    rotation {roll=50}
   }
   n33 = "Transform (Modify) Bone - Bone: dyn_tassel_07_B"!(ComponentPose=@n32, bAlphaBoolEnabled:bool=True, Translation="0,0,0", Rotation="-9.973,0.065,29.258", Scale="1,1,1") -> (Pose) {
    alphaInputType=Bool
    rotationMode=BMM_Additive
    rotationSpace=BCS_ParentBoneSpace
    translationMode=BMM_Additive
    translationSpace=BCS_ParentBoneSpace
    alphaBoolBlend {blendInTime=0.2, blendOutTime=0.2}
    boneToModify {boneName=dyn_tassel_07_B}
    rotation=<rot -9.973 0.065 29.258>
   }
   n34 = "Transform (Modify) Bone - Bone: dyn_tassel_01_A1"!(ComponentPose=@n33, bAlphaBoolEnabled:bool=True, Translation="0,0,0", Rotation="0,0,50", Scale="1,1,1") -> (Pose) {
    alphaInputType=Bool
    rotationMode=BMM_Additive
    rotationSpace=BCS_ParentBoneSpace
    translationMode=BMM_Additive
    translationSpace=BCS_ParentBoneSpace
    alphaBoolBlend {blendInTime=0.2, blendOutTime=0.2}
    boneToModify {boneName=dyn_tassel_01_A1}
    rotation {roll=50}
   }
   n35 = "Transform (Modify) Bone - Bone: dyn_tassel_01_B"!(ComponentPose=@n34, bAlphaBoolEnabled:bool=True, Translation="0,0,0", Rotation="-17.542,1.752,39.217", Scale="1,1,1") -> (Pose) {
    alphaInputType=Bool
    rotationMode=BMM_Additive
    rotationSpace=BCS_ParentBoneSpace
    translationMode=BMM_Additive
    translationSpace=BCS_ParentBoneSpace
    alphaBoolBlend {blendInTime=0.2, blendOutTime=0.2}
    boneToModify {boneName=dyn_tassel_01_B}
    rotation=<rot -17.542 1.752 39.217>
   }
   n36 = "Transform (Modify) Bone - Bone: dyn_tassel_02_A"!(ComponentPose=@n35, bAlphaBoolEnabled:bool=True, Translation="0,0,0", Rotation="0,0,40", Scale="1,1,1") -> (Pose) {
    alphaInputType=Bool
    rotationMode=BMM_Additive
    rotationSpace=BCS_ParentBoneSpace
    translationMode=BMM_Additive
    translationSpace=BCS_ParentBoneSpace
    alphaBoolBlend {blendInTime=0.2, blendOutTime=0.2}
    boneToModify {boneName=dyn_tassel_02_A}
    rotation {roll=40}
   }
   n37 = "Transform (Modify) Bone - Bone: dyn_tassel_02_B"!(ComponentPose=@n36, bAlphaBoolEnabled:bool=True, Translation="0,0,0", Rotation="-27.054,2.861,42.102", Scale="1,1,1") -> (Pose) {
    alphaInputType=Bool
    rotationMode=BMM_Additive
    rotationSpace=BCS_ParentBoneSpace
    translationMode=BMM_Additive
    translationSpace=BCS_ParentBoneSpace
    alphaBoolBlend {blendInTime=0.2, blendOutTime=0.2}
    boneToModify {boneName=dyn_tassel_02_B}
    rotation=<rot -27.054 2.861 42.102>
   }
   n38 = "Transform (Modify) Bone - Bone: dyn_tassel_03_A"!(ComponentPose=@n37, bAlphaBoolEnabled:bool=True, Translation="0,0,0", Rotation="0,0,40", Scale="1,1,1") -> (Pose) {
    alphaInputType=Bool
    rotationMode=BMM_Additive
    rotationSpace=BCS_ParentBoneSpace
    translationMode=BMM_Additive
    translationSpace=BCS_ParentBoneSpace
    alphaBoolBlend {blendInTime=0.2, blendOutTime=0.2}
    boneToModify {boneName=dyn_tassel_03_A}
    rotation {roll=40}
   }
   n39 = "Transform (Modify) Bone - Bone: dyn_tassel_03_B"!(ComponentPose=@n38, bAlphaBoolEnabled:bool=True, Translation="0,0,0", Rotation="-28.235,2.424,39.361", Scale="1,1,1") -> (Pose) {
    alphaInputType=Bool
    rotationMode=BMM_Additive
    rotationSpace=BCS_ParentBoneSpace
    translationMode=BMM_Additive
    translationSpace=BCS_ParentBoneSpace
    alphaBoolBlend {blendInTime=0.2, blendOutTime=0.2}
    boneToModify {boneName=dyn_tassel_03_B}
    rotation=<rot -28.235 2.424 39.361>
   }
   n40 = "Transform (Modify) Bone - Bone: dyn_tassel_04_A"!(ComponentPose=@n39, bAlphaBoolEnabled:bool=True, Translation="0,0,0", Rotation="0,0,50", Scale="1,1,1") -> (Pose) {
    alphaInputType=Bool
    rotationMode=BMM_Additive
    rotationSpace=BCS_ParentBoneSpace
    translationMode=BMM_Additive
    translationSpace=BCS_ParentBoneSpace
    alphaBoolBlend {blendInTime=0.2, blendOutTime=0.2}
    boneToModify {boneName=dyn_tassel_04_A}
    rotation {roll=50}
   }
   n41 = "Transform (Modify) Bone - Bone: dyn_tassel_04_B"!(ComponentPose=@n40, bAlphaBoolEnabled:bool=True, Translation="0,0,0", Rotation="-19.522,0.763,30", Scale="1,1,1") -> (Pose) {
    alphaInputType=Bool
    rotationMode=BMM_Additive
    rotationSpace=BCS_ParentBoneSpace
    translationMode=BMM_Additive
    translationSpace=BCS_ParentBoneSpace
    alphaBoolBlend {blendInTime=0.2, blendOutTime=0.2}
    boneToModify {boneName=dyn_tassel_04_B}
    rotation=<rot -19.522 0.763 30>
   }
   n42 = "Transform (Modify) Bone - Bone: dyn_tassel_05_A"!(ComponentPose=@n41, bAlphaBoolEnabled:bool=True, Translation="0,0,0", Rotation="1.385,9.905,40", Scale="1,1,1") -> (Pose) {
    alphaInputType=Bool
    rotationMode=BMM_Additive
    rotationSpace=BCS_ParentBoneSpace
    translationMode=BMM_Additive
    translationSpace=BCS_ParentBoneSpace
    alphaBoolBlend {blendInTime=0.2, blendOutTime=0.2}
    boneToModify {boneName=dyn_tassel_05_A}
    rotation=<rot 1.385 9.905 40>
   }
   n43 = "Transform (Modify) Bone - Bone: dyn_tassel_05_B"!(ComponentPose=@n42, bAlphaBoolEnabled:bool=True, Translation="0,0,0", Rotation="-26.681,-5.865,18.771", Scale="1,1,1") -> (Pose) {
    alphaInputType=Bool
    rotationMode=BMM_Additive
    rotationSpace=BCS_ParentBoneSpace
    translationMode=BMM_Additive
    translationSpace=BCS_ParentBoneSpace
    alphaBoolBlend {blendInTime=0.2, blendOutTime=0.2}
    boneToModify {boneName=dyn_tassel_05_B}
    rotation=<rot -26.681 -5.865 18.771>
   }
   n44 = "Transform (Modify) Bone - Bone: dyn_tassel_06_A"!(ComponentPose=@n43, bAlphaBoolEnabled:bool=True, Translation="0,0,0", Rotation="0,0,40", Scale="1,1,1") -> (Pose) {
    alphaInputType=Bool
    rotationMode=BMM_Additive
    rotationSpace=BCS_ParentBoneSpace
    translationMode=BMM_Additive
    translationSpace=BCS_ParentBoneSpace
    alphaBoolBlend {blendInTime=0.2, blendOutTime=0.2}
    boneToModify {boneName=dyn_tassel_06_A}
    rotation {roll=40}
   }
   n45 = "Transform (Modify) Bone - Bone: dyn_tassel_06_B"!(ComponentPose=@n44, bAlphaBoolEnabled:bool=True, Translation="0,0,0", Rotation="-25.447,-11.334,30", Scale="1,1,1") -> (Pose) {
    alphaInputType=Bool
    rotationMode=BMM_Additive
    rotationSpace=BCS_ParentBoneSpace
    translationMode=BMM_Additive
    translationSpace=BCS_ParentBoneSpace
    alphaBoolBlend {blendInTime=0.2, blendOutTime=0.2}
    boneToModify {boneName=dyn_tassel_06_B}
    rotation=<rot -25.447 -11.334 30>
   }
   n46 = "Transform (Modify) Bone - Bone: dyn_tassel_07_A"(ComponentPose=@n45, bAlphaBoolEnabled=True, Translation="0,0,0", Rotation="0,0,3", Scale="1,1,1") {
    alphaInputType=Bool
    rotationMode=BMM_Additive
    rotationSpace=BCS_ParentBoneSpace
    translationMode=BMM_Additive
    translationSpace=BCS_ParentBoneSpace
    alphaBoolBlend {blendInTime=0.2, blendOutTime=0.2}
    boneToModify {boneName=dyn_tassel_07_A}
    rotation {roll=3}
   }
   n47 = "Transform (Modify) Bone - Bone: dyn_tassel_01_A1"(ComponentPose=@n46, bAlphaBoolEnabled=True, Translation="0,0,0", Rotation="0,0,4", Scale="1,1,1") {
    alphaInputType=Bool
    rotationMode=BMM_Additive
    rotationSpace=BCS_ParentBoneSpace
    translationMode=BMM_Additive
    translationSpace=BCS_ParentBoneSpace
    alphaBoolBlend {blendInTime=0.2, blendOutTime=0.2}
    boneToModify {boneName=dyn_tassel_01_A1}
    rotation {roll=4}
   }
   n48 = "Transform (Modify) Bone - Bone: dyn_tassel_02_A"(ComponentPose=@n47, bAlphaBoolEnabled=True, Translation="0,0,0", Rotation="0,0,3", Scale="1,1,1") {
    alphaInputType=Bool
    rotationMode=BMM_Additive
    rotationSpace=BCS_ParentBoneSpace
    translationMode=BMM_Additive
    translationSpace=BCS_ParentBoneSpace
    alphaBoolBlend {blendInTime=0.2, blendOutTime=0.2}
    boneToModify {boneName=dyn_tassel_02_A}
    rotation {roll=3}
   }
   n49 = "Transform (Modify) Bone - Bone: dyn_tassel_03_A"(ComponentPose=@n48, bAlphaBoolEnabled=True, Translation="0,0,0", Rotation="0,0,3", Scale="1,1,1") {
    alphaInputType=Bool
    rotationMode=BMM_Additive
    rotationSpace=BCS_ParentBoneSpace
    translationMode=BMM_Additive
    translationSpace=BCS_ParentBoneSpace
    alphaBoolBlend {blendInTime=0.2, blendOutTime=0.2}
    boneToModify {boneName=dyn_tassel_03_A}
    rotation {roll=3}
   }
   n50 = "Transform (Modify) Bone - Bone: dyn_tassel_04_A"(ComponentPose=@n49, bAlphaBoolEnabled=True, Translation="0,0,0", Rotation="0,0,5", Scale="1,1,1") {
    alphaInputType=Bool
    rotationMode=BMM_Additive
    rotationSpace=BCS_ParentBoneSpace
    translationMode=BMM_Additive
    translationSpace=BCS_ParentBoneSpace
    alphaBoolBlend {blendInTime=0.2, blendOutTime=0.2}
    boneToModify {boneName=dyn_tassel_04_A}
    rotation {roll=5}
   }
   n51 = "Transform (Modify) Bone - Bone: dyn_tassel_05_A"(ComponentPose=@n50, bAlphaBoolEnabled=True, Translation="0,0,0", Rotation="0,0,5", Scale="1,1,1") {
    alphaInputType=Bool
    rotationMode=BMM_Additive
    rotationSpace=BCS_ParentBoneSpace
    translationMode=BMM_Additive
    translationSpace=BCS_ParentBoneSpace
    alphaBoolBlend {blendInTime=0.2, blendOutTime=0.2}
    boneToModify {boneName=dyn_tassel_05_A}
    rotation {roll=5}
   }
   n52 = "Transform (Modify) Bone - Bone: dyn_tassel_06_A"(ComponentPose=@n51, bAlphaBoolEnabled=True, Translation="0,0,0", Rotation="0,0,3", Scale="1,1,1") {
    alphaInputType=Bool
    rotationMode=BMM_Additive
    rotationSpace=BCS_ParentBoneSpace
    translationMode=BMM_Additive
    translationSpace=BCS_ParentBoneSpace
    alphaBoolBlend {blendInTime=0.2, blendOutTime=0.2}
    boneToModify {boneName=dyn_tassel_06_A}
    rotation {roll=3}
   }
   Output Pose(Result=Component To Local(ComponentPose=@n52)) {layerGroup=None}
  ]}
 EventGraph {nodes [
   Event BlueprintUpdateAnimation()
   > Set IsinAir(IsinAir:bool=OR Boolean(A=Get bIsSkydiveFloating(), B=Get bIsSkydiveDiveMode())) -> (Output_Get:bool)
   n1 = Event OnCharacterCustomizationLoaded()
   n2 = Get Array Item with EFort Custom Part Type(Array:Array of Custom Character Part Object References=@n1.CustomParts) -> (Output:Custom Character Part Object Reference)
   > Set isCharFemale(isCharFemale:bool="Equal (Enum)"(
     A:EFortCustomGender Enum=Get GenderPermitted!(self:Custom Character Part Object Reference=@n2) -> (GenderPermitted:EFortCustomGender Enum=Invalid)
     B:EFortCustomGender Enum=Female
    )) -> (Output_Get:bool)
   TryGetPawnOwner()
  ]}
]
variables [
 Gender Permitted {type=EFortCustomGender, default=Invalid}
 isCharFemale {type=boolean}
 IsinAir {type=boolean}
]
